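# Idle Niagara effect for the SwampFish backpack, filling three jars on the pack with glowing fireflies and flickering cyan lights.
# Three burst groups follow skeletal mesh sample points, with curl noise and a pendulum constraint confining their motion to each jar's height and radius.
# A system timeline raises agitation and glow, with time of day setting base brightness.
# Waveform-driven lights share that timeline; debug and backup emitters are disabled.
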NiagaraSystem /BRCosmetics/Accessories/FORT_Backpacks/Backpack_M_MED_SwampFish/FX/NS_SwampFish_Backpack_Idle
dependencies {
 usedDataInterfaces [
  &NiagaraDataInterfaceSkeletalMesh
  &NiagaraDataInterfaceCurve
  &NiagaraDataInterfaceArrayInt32
  &NiagaraDataInterfaceArrayFloat
 ]
 usedDynamicInputs [&SetBoolByFloatComparison]
 usedModules [
  &self:SystemSpawnScript.SystemScriptSource.SystemScriptGraph.NiagaraNodeAssignment_0.SetVariables_#1
  &SystemState
  &self:SystemSpawnScript.SystemScriptSource.SystemScriptGraph.NiagaraNodeAssignment_2.SetVariables_#2
  &Timeline
  &self:SystemSpawnScript.SystemScriptSource.SystemScriptGraph.NiagaraNodeAssignment_1.SetVariables_#3
 ]
}
emitters {
 FN_Debug {
  bEnabled=false
  emitterUpdateScript [
   &EmitterState()
   &SpawnBurst_Instantaneous(Spawn Count=1, Use Spawn Probability=false)
  ]
  particleSpawnScript [/Niagara/Modules/Spawn/Initialization/InitializeParticle(
    Lifetime=5
    Module.Write Color=true
    Module.Write Lifetime=true
    Module.Write Mass=true
    Module.Write Material Random=true
    Module.Write Position=false
    Module.Write Scale=false
    Module.Write SpriteRotation=false
    Module.Write SpriteSize=true
    Position=&SimulationPosition()
    Sprite Size=<100 100>
   )]
  particleUpdateScript [
   &ParticleState()
   &FN_Timer()
   &SkeletalMeshLocation(
    MaskSpawnGroup=false
    Mesh Sampling Type=::NewEnumerator5
    Position Sampling=::NewEnumerator2
    Sampled Position Offset Coordinate Space:ENiagaraSkelMeshTransforms=::NewEnumerator1
    Skeletal Mesh=DataInterface() {propertyValues {
      DefaultMesh=null
      ExcludeBoneName=root
      OptionalComponentClass=null
      PreviewMesh=/BRCosmetics/Characters/Player/Male/Medium/Bodies/M_MED_Swamp_Fish/Meshes/M_MED_Swamp_Fish
      SkinningMode=SkinOnTheFly
      SoftSourceActor=null
      SourceMode=Default
      UvSetIndex=0
      WholeMeshLOD=-1
      bExcludeBone=false
      bReadDeformedGeometry=true
      bRequireCurrentFrameData=true
      ComponentTags []
      FilteredBones []
      FilteredSockets [FX_Fireflies_Socket]
      MeshUserParameter {parameter {name=None, typeDefHandle {registeredTypeIndex=1}, varData []}}
      SamplingRegions []
     }}
    Socket Sampling Mode=::NewEnumerator5
   )
   &SampleSkeletalMesh(
    AdditionalYawPitchRoll=false
    Mesh Sampling Type:ENiagaraSkelSamplingModeFull=::NewEnumerator5
    Skeletal Mesh=DataInterface() {propertyValues {
      DefaultMesh=null
      ExcludeBoneName=root
      OptionalComponentClass=null
      PreviewMesh=&M_MED_Swamp_Fish
      SkinningMode=SkinOnTheFly
      SoftSourceActor=null
      SourceMode=Default
      UvSetIndex=0
      WholeMeshLOD=-1
      bExcludeBone=false
      bReadDeformedGeometry=true
      bRequireCurrentFrameData=true
      ComponentTags []
      FilteredBones []
      FilteredSockets [FX_Fireflies_Socket]
      MeshUserParameter {parameter {name=None, typeDefHandle {registeredTypeIndex=1}, varData []}}
      SamplingRegions []
     }}
    Socket Sampling Mode:ENiagaraSocketSamplingMode=::NewEnumerator5
   )
   /SetVariables(Particles.Debug:NiagaraFloat=&MakeFloatFromVector(Vector=$Particles.SampleSkeletalMesh.SampledPosition, Channel=::NewEnumerator0)) {isSetParameters=true}
   &DynamicMaterialParameters(
    Index 0 Param 1=$Particles.Debug
    Module.Param0WriteEnabled=true
    Module.Param1WriteEnabled=true
    Module.Param2WriteEnabled=true
    Module.Param3WriteEnabled=true
   )
   &CameraOffset(Camera Offset Amount=111)
  ]
  rendererClasses [&NiagaraSpriteRendererProperties]
  renderers [{rendererClass=&NiagaraSpriteRendererProperties}]
 }
 Fireflies {
  emitterUpdateScript [
   &EmitterState(
    Inactive Response=::NewEnumerator1
    Life Cycle Mode=::NewEnumerator1
    Loop Behavior=::NewEnumerator1
    Loop Duration Mode=::NewEnumerator1
    UseLoopDelay=false
   )
   &SpawnBurst_Instantaneous(
    Spawn Count=&SelectIntFromArray(
     Array Sampling Mode=::NewEnumerator1
     Direct Array Index=?
     Int Selection Array=DataInterface() {propertyValues {IntData [3,4,5,6,7]}}
     Random Seed=0
    )
    Use Spawn Probability=false
    UseLoopCountLimit=false
   )
   &SpawnBurst_Instantaneous(
    Spawn Count=&SelectIntFromArray(
     Array Sampling Mode=::NewEnumerator1
     Direct Array Index=?
     Int Selection Array=DataInterface() {propertyValues {IntData [3,4,5,6,7]}}
     Random Seed=0
    )
    Spawn Group=1
    Use Spawn Probability=false
    UseLoopCountLimit=false
   ) {module=SpawnBurst_Instantaneous001}
   &SpawnBurst_Instantaneous(
    Spawn Count=&SelectIntFromArray(
     Array Sampling Mode=::NewEnumerator1
     Direct Array Index=?
     Int Selection Array=DataInterface() {propertyValues {IntData [3,4,5,6,7]}}
     Random Seed=0
    )
    Spawn Group=2
    Use Spawn Probability=false
    UseLoopCountLimit=false
   ) {module=SpawnBurst_Instantaneous002}
  ]
  particleSpawnScript [
   /Niagara/Modules/Spawn/Initialization/V2/InitializeParticle(
    Color Mode=::NewEnumerator0
    Mesh Scale Mode=::NewEnumerator1
    Mesh Uniform Scale Max=0.333
    Mesh Uniform Scale Min=0.266
    Sprite Size Mode=::NewEnumerator3
    Uniform Sprite Size=9
    UsePositionOffset=false
    Write Material Random=true
   )
   &PartitionParticles(ID to Partition:NiagaraInt32=&SpawnGroup(), Number Of Partitions=3)
  ]
  particleUpdateScript [
   &ParticleState(Kill Particles When Lifetime Has Elapsed=false, Let Infinitely Lived Particles Die When Emitter Deactivates=true)
   /SetVariables(Particles.ReactivityFadeInValue:NiagaraFloat=&FloatFromCurve(
     FloatCurve=DataInterface() {propertyValues {Curve {keys [<at 0 0>,<at 1 1>]}}}
     CurveIndex=&Multiply_Float(A=$System.Timeline.NormalizedTime, B=&RandomRangeFloat(Minimum=1, Maximum=7))
    )) {isSetParameters=true}
   /SetVariables(Particles.ReactivityFadeInValueColors:NiagaraFloat=&FNTimeOfDayScalar(
     Night Scale=&Lerp_Float(A=0.3, B=1, Alpha=$Particles.ReactivityFadeInValue)
     Morning Scale=$Particles.ReactivityFadeInValue
     Day Scale=$Particles.ReactivityFadeInValue
     Evening Scale=$Particles.ReactivityFadeInValue
    )) {enabled=false, isSetParameters=true}
   /SetVariables(Particles.ReactivityFadeInValueColors:NiagaraFloat=&Lerp_Float(
     A=&FNTimeOfDayScalar(Night Scale=0.3, Morning Scale=0.4, Day Scale=0.35, Evening Scale=0.35)
     B=1
     Alpha=$Particles.ReactivityFadeInValue
    )) {isSetParameters=true}
   &SampleSkeletalMesh(
    AdditionalYawPitchRoll=false
    Bone / Socket Index:NiagaraInt32=$Particles.PartitionParticles.Partition
    Bone Sampling Mode:ENiagaraBoneSamplingMode=::NewEnumerator4
    Skeletal Mesh:NiagaraDataInterfaceSkeletalMesh=$System.Skeletal Mesh
   )
   /SetVariables(Particles.JarVector:Vector3f=&RotateVectorByQuaternion(Vector To Rotate=<1 0 0>, Quaternion=$Particles.SampleSkeletalMesh.SampledRotation)) {isSetParameters=true, module=SetVariables_#4}
   /SetVariables(Particles.PositionCapsuleOffset:NiagaraPosition=&AddVectorToPosition(
     Position=$Particles.SampleSkeletalMesh.SampledPosition
     Vector=&Multiply_VectorByFloat(Vector=$Particles.JarVector, Float=&RandomRangeFloat(Minimum=0, Maximum=$User.JarHeight))
    )) {isSetParameters=true}
   /SetVariables(
    Particles.FireflyMeshCameraOffset:NiagaraFloat=&Lerp_Float(A=0, B=-0.03, Alpha=$Particles.ReactivityFadeInValueColors)
    Particles.FireflyMeshColor:LinearColor=&MakeLinearColorFromVectorAndFloat(
     "Vector (RGB)"=&Lerp_Vector(A=<0.6 0.6 0.6>, B=<4 4 4>, Alpha=$Particles.ReactivityFadeInValueColors)
     "Float (Alpha)"=&Lerp_Float(A=0.5, B=1, Alpha=$Particles.ReactivityFadeInValueColors)
    )
    Particles.GlowSpriteCameraOffset:NiagaraFloat=&Lerp_Float(A=-0.03, B=0.01, Alpha=$Particles.ReactivityFadeInValueColors)
    Particles.GlowSpriteColor:LinearColor=&MultiplyLinearColor(
     Linear Color=<rgba 0 1 0.7915 1>
     Scale Factor=&MakeVector4_FromFloat(Float=&Lerp_Float(
       A=0.133
       B=&Multiply_Float(
        A=0.5
        B=&Waveform(
         "Waveform [1]"=::NewEnumerator0
         "[1] Frequency"=&RandomRangeFloat(Minimum=0.4, Maximum=0.8)
         "[1] Amplitude Min/Max"=<0.6 1>
        )
       )
       Alpha=$Particles.ReactivityFadeInValueColors
      ))
    )
   ) {isSetParameters=true}
   &DynamicMaterialParameters(
    Index 0 Param 1=3
    Index 0 Param 2=9
    Index 0 Param 3=$System.Timeline.Time
    Module.Index0Param0Write=true
    Module.Index0Param1Write=true
    Module.Index0Param2Write=true
    Module.Index0Param3Write=false
   )
   &CurlNoiseForce(
    Noise Frequency=&RandomRangeFloat(Minimum=44, Maximum=77)
    Noise Strength=&Lerp_Float(A=222, B=444, Alpha=$Particles.ReactivityFadeInValue)
    Pan Noise=true
    Pan Noise Field=&MakeVector(X=0, Y=0, Z=&Lerp_Float(A=1, B=3.5, Alpha=$Particles.ReactivityFadeInValue))
    Scale Curl Lookup By Engine Owner Scale=false
   )
   &CurlNoiseForce(
    Noise Frequency=&RandomRangeFloat(Minimum=60, Maximum=100)
    Noise Strength=&Lerp_Float(A=0, B=&RandomRangeFloat(Minimum=800, Maximum=1100), Alpha=$Particles.ReactivityFadeInValue)
    Pan Noise=true
    Pan Noise Field=&MakeVector(X=3, Y=4, Z=5)
    Scale Curl Lookup By Engine Owner Scale=false
   ) {module=CurlNoiseForce001}
   &PendulumSetup(
    Module.Enable Spring Driver=false
    Pendulum Length:NiagaraFloat=$User.JarRadius
    Pendulum Pivot Offset:NiagaraPosition=$Particles.PositionCapsuleOffset
    Pendulum Pivot Offset Coordinate Space:ENiagaraCoordinateSpace=::Local
    Rigid Pendulum=false
   )
   &SolveForcesAndVelocity(Clamp Velocity=false, Limit Acceleration=false)
   &PendulumConstraint()
   &CalculateAccurateVelocity()
  ]
  rendererClasses [&NiagaraMeshRendererProperties, &NiagaraSpriteRendererProperties]
  renderers [
   {rendererClass=&NiagaraMeshRendererProperties, rendererIndex=0}
   {rendererClass=&NiagaraSpriteRendererProperties, rendererIndex=1}
   {rendererClass=&NiagaraMeshRendererProperties, rendererIndex=2}
   {rendererClass=&NiagaraSpriteRendererProperties, rendererIndex=3}
  ]
 }
 Fireflies_backup {
  bEnabled=false
  emitterUpdateScript [
   &EmitterState()
   &SpawnRate(SpawnRate=8, Use Spawn Probability=false)
  ]
  particleSpawnScript [
   &InitializeParticle(
    Color Maximum=<rgba 0 3 1.5 1>
    Color Minimum=<rgba 0 2 3 1>
    Color Mode=::NewEnumerator2
    Lifetime=&RandomRangeFloat(Minimum=1, Maximum=2)
    Mesh Scale Mode=::NewEnumerator1
    Mesh Uniform Scale Max=0.3
    Mesh Uniform Scale Min=0.2
    Sprite Size Mode=::NewEnumerator1
    Uniform Sprite Size Max=3
    Uniform Sprite Size Min=1
    UsePositionOffset=false
    Write Material Random=true
   )
   &PartitionParticles(Number Of Partitions=3)
   &CameraOffset(Camera Offset Amount=5) {enabled=false}
   &SkeletalMeshLocation(
    Bone / Socket Index=$Particles.PartitionParticles.Partition
    MaskSpawnGroup=false
    Mesh Sampling Type=::NewEnumerator5
    Position Sampling=::NewEnumerator2
    Sampled Position Offset Coordinate Space:ENiagaraSkelMeshTransforms=::NewEnumerator1
    Skeletal Mesh=DataInterface() {propertyValues {
      DefaultMesh=null
      ExcludeBoneName=root
      OptionalComponentClass=null
      PreviewMesh=/BRCosmetics/Accessories/FORT_Backpacks/Backpack_M_MED_SwampFish/Meshes/M_MED_SwampFish_Pack
      SkinningMode=SkinOnTheFly
      SoftSourceActor=null
      SourceMode=Default
      UvSetIndex=0
      WholeMeshLOD=-1
      bExcludeBone=false
      bReadDeformedGeometry=true
      bRequireCurrentFrameData=true
      ComponentTags []
      FilteredBones []
      FilteredSockets [FX_Fireflies_Socket_Bottom, FX_Fireflies_Socket_Top_L, FX_Fireflies_Socket_Top_R]
      MeshUserParameter {parameter {name=None, typeDefHandle {registeredTypeIndex=1}, varData []}}
      SamplingRegions []
     }}
    Socket Sampling Mode=::NewEnumerator5
   ) {module=SkeletalMeshLocation001}
   &ShapeLocation(
    MaskSpawnGroup=false
    Offset Mode=::NewEnumerator0
    Override Seed=false
    Shape Origin=$Particles.SkeletalMeshLocation001.SampledPosition
    Sphere Radius=15
    UseOwnerScale=true
   )
   &InitialMeshOrientation(Use Rotation Vector=false)
  ]
  particleUpdateScript [
   &SkeletalMeshLocation(
    Bone / Socket Index=$Particles.PartitionParticles.Partition
    MaskSpawnGroup=false
    Mesh Sampling Type=::NewEnumerator5
    Position Sampling=::NewEnumerator2
    Sampled Position Offset Coordinate Space:ENiagaraSkelMeshTransforms=::NewEnumerator1
    Skeletal Mesh=DataInterface() {propertyValues {
      DefaultMesh=null
      ExcludeBoneName=root
      OptionalComponentClass=null
      PreviewMesh=&M_MED_SwampFish_Pack
      SkinningMode=SkinOnTheFly
      SoftSourceActor=null
      SourceMode=Default
      UvSetIndex=0
      WholeMeshLOD=-1
      bExcludeBone=false
      bReadDeformedGeometry=true
      bRequireCurrentFrameData=true
      ComponentTags []
      FilteredBones []
      FilteredSockets [FX_Fireflies_Socket_Bottom, FX_Fireflies_Socket_Top_L, FX_Fireflies_Socket_Top_R]
      MeshUserParameter {parameter {name=None, typeDefHandle {registeredTypeIndex=1}, varData []}}
      SamplingRegions []
     }}
    Socket Sampling Mode=::NewEnumerator5
   )
   &ParticleState()
   &ScaleMeshSize(Scale Factor=&VectorFromFloat(Value=&FloatFromCurve(FloatCurve=DataInterface() {propertyValues {Curve {keys [<at 0 0>,<at 0.2 1>,<at 0.8 1>,<at 1 0>]}}})))
   /SetVariables(Particles.PositionCapsuleOffset:NiagaraPosition=&AddVectorToPosition(
     Position=&TransformPosition(Position=$Particles.SkeletalMeshLocation.SampledPosition, Source Space=::World, Destination Space=::Local)
     Vector=&RandomRangeVector(Minimum=<-2 -2 -3>, Maximum=<2 2 3>)
    )) {isSetParameters=true, module=SetVariables_#4}
   &PendulumSetup(
    Module.Enable Spring Driver=false
    Pendulum Length=2
    Pendulum Pivot Offset:NiagaraPosition=$Particles.PositionCapsuleOffset
    Pendulum Pivot Offset Coordinate Space:ENiagaraCoordinateSpace=::Local
    Rigid Pendulum=false
   )
   &CurlNoiseForce(
    Noise Frequency=&RandomRangeFloat(Minimum=80, Maximum=150)
    Noise Strength=&RandomRangeFloat(Minimum=200, Maximum=400)
    Pan Noise=false
   )
   &CurlNoiseForce(
    Noise Frequency=&RandomRangeFloat(Minimum=0.02, Maximum=0.1)
    Noise Strength=&RandomRangeFloat(Minimum=100, Maximum=300)
    Pan Noise=false
   ) {module=CurlNoiseForce001}
   &SolveForcesAndVelocity(Clamp Velocity=false, Limit Acceleration=false)
   &ScaleColor(
    Scale Alpha=&FloatFromCurve(FloatCurve=DataInterface() {propertyValues {Curve {keys [<at 0 0>,<at 0.4 1>,<at 0.8 1>,<at 1 0>]}}})
    Scale RGB=&Multiply_VectorByFloat(
     Vector=<1 1 1>
     Float=&Lerp_Float(
      A=1
      B=20
      Alpha=&FN_Select_Float_from_Time_Of_Day(Morning=$System.Timeline.NormalizedTime, Day=$System.Timeline.NormalizedTime, Evening=$System.Timeline.NormalizedTime, Night=1)
     )
    )
    ScaleA=true
    ScaleRGB=true
   )
   &PendulumConstraint()
   &CalculateAccurateVelocity()
   &DynamicMaterialParameters(
    Index 0 Param 1=3
    Index 0 Param 4=2
    Module.Index0Param0Write=true
    Module.Index0Param1Write=false
    Module.Index0Param2Write=false
    Module.Index0Param3Write=true
   )
  ]
  rendererClasses [&NiagaraSpriteRendererProperties, &NiagaraMeshRendererProperties]
  renderers [
   {rendererClass=&NiagaraSpriteRendererProperties}
   {rendererClass=&NiagaraSpriteRendererProperties, rendererIndex=1}
   {rendererClass=&NiagaraMeshRendererProperties, rendererIndex=2}
  ]
 }
 Light {
  emitterUpdateScript [
   &EmitterState(
    Inactive Response=::NewEnumerator1
    Life Cycle Mode=::NewEnumerator1
    Loop Behavior=::NewEnumerator1
    Loop Duration Mode=::NewEnumerator1
    UseLoopDelay=false
   )
   &SpawnBurst_Instantaneous(Spawn Count=3, Use Spawn Probability=false, UseLoopCountLimit=false)
  ]
  particleSpawnScript [
   &InitializeParticle(Color Mode=::NewEnumerator0, Position=$User.JarOffset, Position Mode=::NewEnumerator1, UsePositionOffset=false, Write Material Random=true)
   &PartitionParticles(Number Of Partitions=3)
  ]
  particleUpdateScript [
   &SkeletalMeshLocation(
    Bone / Socket Index=$Particles.PartitionParticles.Partition
    Bone Sampling Mode=::NewEnumerator4
    MaskSpawnGroup=false
    Sampled Position Offset=<5 0 0>
    Skeletal Mesh=$System.Skeletal Mesh
   )
   &ParticleState(Kill Particles When Lifetime Has Elapsed=false)
   /SetVariables(Particles.Color=&MultiplyLinearColor(
     Linear Color=<rgba 0 9 7.7913 1>
     Scale Factor=&MakeVector4_FromFloat(Float=&Multiply_Float(
       A=&SelectFloatFromArray(Direct Array Index=$Particles.PartitionParticles.Partition, Float Selection Array=DataInterface() {propertyValues {FloatData [0.7,0.3,0.3]}})
       B=&Multiply_Float(
        A=$System.Timeline.Time
        B=&Waveform(
         Waveform Count=::NewEnumerator2
         Global Input Position=&RandomRangeFloat(Minimum=0, Maximum=1)
         "Waveform [1]"=::NewEnumerator0
         "[1] Frequency"=1
         "[1] Amplitude Min/Max"=<0.333 1>
         "[2] Average"=false
         "[2] Input Position"=0
         "[2] Phase"=$Emitter.Age
         "[2] Phase Scale"=1
         "[2] Compound Frequency"=2
         "[2] Frequency"=1.618
         "[2] Amplitude Min/Max"=<0.333 1>
         "[2] Scale"=1
         "[2] Bias"=0
         "[2] Exponent"=1
         "[2] Pulse Duration"=0.1
         "[2] Clip Triangle Min/Max"=<0 0>
         "[2] Random Seed"=0
         "[3] Average"=false
         "[3] Input Position"=0
         "[3] Phase"=$Emitter.Age
         "[3] Phase Scale"=1
         "[3] Frequency"=2.618
         "[3] Compound Frequency"=2
         "[3] Amplitude Min/Max"=<0.666 1>
         "[3] Scale"=1
         "[3] Bias"=0
         "[3] Exponent"=1
         "[3] Pulse Duration"=0.1
         "[3] Clip Triangle Min/Max"=<0 0>
         "[3] Random Seed"=0
        )
       )
      ))
    )) {isSetParameters=true}
  ]
  rendererClasses [&NiagaraLightRendererProperties]
  renderers [{rendererClass=&NiagaraLightRendererProperties, rendererIndex=0}]
 }
}
summary {
 systemName=NS_SwampFish_Backpack_Idle
 emitters [
  {bEnabled=false, emitterName=FN_Debug, rendererClasses [&NiagaraSpriteRendererProperties]}
  {
   bEnabled=false
   emitterName=Fireflies_backup
   rendererClasses [&NiagaraSpriteRendererProperties, &NiagaraMeshRendererProperties]
  }
  {emitterName=Fireflies, rendererClasses [&NiagaraMeshRendererProperties, &NiagaraSpriteRendererProperties]}
  {emitterName=Light, rendererClasses [&NiagaraLightRendererProperties]}
 ]
 userVariables [
  User.bIsFrontEnd {type=NiagaraBool}
  User.bIsFrontEndPreview {type=NiagaraBool}
  User.JarHeight {type=NiagaraFloat, defaultValue=5, description=""}
  User.JarOffset {type=NiagaraPosition, description="", defaultValue {}}
  User.JarRadius {type=NiagaraFloat, defaultValue=2.15, description=""}
  User.Reactivity {type=NiagaraBool}
 ]
}
