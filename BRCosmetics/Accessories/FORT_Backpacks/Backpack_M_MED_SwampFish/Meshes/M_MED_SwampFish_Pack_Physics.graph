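# Physics asset for the SwampFish backpack cosmetic, defining capsule bodies and joints for its dangling cloth and jars.
# The cloth chain runs from a master bone through four segments, then forks into left and right branches.
# Two top jar chains and one bottom jar chain hang from their own tiny master bodies, with swing-limited, low-twist joints.
# An unconstrained pelvis body with two crossed capsules serves as a body collider.
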PhysicsAsset /BRCosmetics/Accessories/FORT_Backpacks/Backpack_M_MED_SwampFish/Meshes/M_MED_SwampFish_Pack_Physics
bodies {
 dyn_cloth_01 [{shapeType=Capsule, length=11.5, radius=3.5, shapeName=dyn_cloth_01_capsule, rotation=<rot 0 0 90>}]
 dyn_cloth_02 [{shapeType=Capsule, length=16.5, radius=2.5, shapeName=dyn_cloth_02_capsule, rotation=<rot 0 0 90>}]
 dyn_cloth_03 [{shapeType=Capsule, length=18, radius=3, shapeName=dyn_cloth_03_capsule, rotation=<rot 0 0 90>}]
 dyn_cloth_04 [{shapeType=Capsule, length=11.5, radius=3, shapeName=dyn_cloth_04_capsule, rotation=<rot 0 0 90>}]
 dyn_cloth_l_05 [{shapeType=Capsule, length=5.5, radius=2.5, shapeName=dyn_cloth_l_05_capsule, center=<1.5 0 0>, rotation=<rot 90 0 0>}]
 dyn_cloth_l_06 [{shapeType=Capsule, length=7, radius=3.5, shapeName=dyn_cloth_l_06_capsule, center=<2 0 0>, rotation=<rot 90 0 0>}]
 dyn_cloth_l_07 [{shapeType=Capsule, length=2.5, radius=3.005, shapeName=dyn_cloth_l_07_capsule, center=<2 0 0>, rotation=<rot 90 0 0>}]
 dyn_cloth_mstr [{shapeType=Capsule, length=0.505, radius=0.505, shapeName=dyn_cloth_mstr_capsule, rotation=<rot 0 -90 0>}]
 dyn_cloth_r_05 [{shapeType=Capsule, length=6, radius=2, shapeName=dyn_cloth_r_05_capsule, center=<1.5 0 0>, rotation=<rot 90 0 0>}]
 dyn_cloth_r_06 [{shapeType=Capsule, length=4, radius=3, shapeName=dyn_cloth_r_06_capsule, center=<2 0 0>, rotation=<rot 90 0 0>}]
 dyn_jar_bottom_c_01 [{shapeType=Capsule, length=13.5, radius=0.5, shapeName=dyn_jar_bottom_c_01_capsule, center=<7.429 0 0>, rotation=<rot 90 0 0>}]
 dyn_jar_bottom_c_02 [{shapeType=Capsule, length=2.5, radius=1, shapeName=dyn_jar_bottom_c_02_capsule, rotation=<rot 90 0 0>}]
 dyn_jar_bottom_c_03 [{shapeType=Capsule, length=7, radius=4, shapeName=dyn_jar_bottom_c_03_capsule, center=<3 -1.1 0>, rotation=<rot 90 0 0>}]
 dyn_jar_bottom_c_mstr [{shapeType=Capsule, length=0.505, radius=0.505, shapeName=dyn_jar_bottom_c_mstr_capsule, rotation=<rot 0 -90 0>}]
 dyn_jar_top_l_01 [{
   shapeType=Capsule
   length=3.035
   radius=1.4157
   shapeName=dyn_jar_top_l_01_capsule
   center=<3.0246 -1.2123 -0.2965>
   rotation=<rot 0 -4.5852 -88.3339>
  }]
 dyn_jar_top_l_02 [{shapeType=Capsule, length=4.5868, radius=4.795, shapeName=dyn_jar_top_l_02_capsule, center=<2.5 0 0>, rotation=<rot 0 90 -80>}]
 dyn_jar_top_l_mstr [{shapeType=Capsule, length=0.505, radius=0.505, shapeName=dyn_jar_top_l_mstr_capsule, rotation=<rot 0 -90 0>}]
 dyn_jar_top_r_01 [{
   shapeType=Capsule
   length=2.1727
   radius=1.5585
   shapeName=dyn_jar_top_r_01_capsule
   center=<2.7329 0.2684 -0.2462>
   rotation=<rot 0 -2.8918 -85.6093>
  }]
 dyn_jar_top_r_02 [{shapeType=Capsule, length=4.9265, radius=4.1821, shapeName=dyn_jar_top_r_02_capsule, center=<2.25 0 0>, rotation=<rot 0 90 -85>}]
 dyn_jar_top_r_mstr [{shapeType=Capsule, length=0.505, radius=0.505, shapeName=dyn_jar_top_r_mstr_capsule, rotation=<rot 0 -90 0>}]
 pelvis_collider [
  {shapeType=Capsule, length=28.5, radius=10.5, shapeName=dyn_buttCollider_capsule, center=<-0.75 0 0>, rotation=<rot 0 -90 0>}
  {shapeType=Capsule, length=19.5, radius=11.5, shapeName=dyn_buttCollider_capsule, rotation=<rot 90 0 0>}
 ]
}
constraints [
 {bone1Name=dyn_jar_top_l_01, bone2Name=dyn_jar_top_l_mstr, swing1LimitDegrees=35, swing2LimitDegrees=25, twistLimitDegrees=2}
 {bone1Name=dyn_jar_top_l_02, bone2Name=dyn_jar_top_l_01, swing1LimitDegrees=40, swing2LimitDegrees=40, twistLimitDegrees=2}
 {bone1Name=dyn_jar_top_r_01, bone2Name=dyn_jar_top_r_mstr, swing1LimitDegrees=40, swing2LimitDegrees=40, twistLimitDegrees=2}
 {bone1Name=dyn_jar_top_r_02, bone2Name=dyn_jar_top_r_01, swing1LimitDegrees=40, swing2LimitDegrees=40, twistLimitDegrees=2}
 {bone1Name=dyn_cloth_l_06, bone2Name=dyn_cloth_l_05, swing1LimitDegrees=10, swing2LimitDegrees=45, twistLimitDegrees=5}
 {bone1Name=dyn_cloth_r_06, bone2Name=dyn_cloth_r_05, swing1LimitDegrees=10, swing2LimitDegrees=30, twistLimitDegrees=5}
 {bone1Name=dyn_jar_bottom_c_01, bone2Name=dyn_jar_bottom_c_mstr, swing1LimitDegrees=25, swing2LimitDegrees=10, twistLimitDegrees=5}
 {bone1Name=dyn_jar_bottom_c_02, bone2Name=dyn_jar_bottom_c_01, swing1LimitDegrees=40, swing2LimitDegrees=40, twistLimitDegrees=5}
 {bone1Name=dyn_jar_bottom_c_03, bone2Name=dyn_jar_bottom_c_02, swing1LimitDegrees=60, swing2LimitDegrees=60, twistLimitDegrees=5}
 {bone1Name=dyn_cloth_l_05, bone2Name=dyn_cloth_04, swing1LimitDegrees=15, swing2LimitDegrees=15, twistLimitDegrees=1}
 {bone1Name=dyn_cloth_04, bone2Name=dyn_cloth_03, swing1LimitDegrees=30, swing2LimitDegrees=30, twistLimitDegrees=1}
 {bone1Name=dyn_cloth_03, bone2Name=dyn_cloth_02, swing1LimitDegrees=20, swing2LimitDegrees=20, twistLimitDegrees=1}
 {bone1Name=dyn_cloth_02, bone2Name=dyn_cloth_01, swing1LimitDegrees=15, swing2LimitDegrees=15, twistLimitDegrees=1}
 {bone1Name=dyn_cloth_01, bone2Name=dyn_cloth_mstr, swing1LimitDegrees=15, swing2LimitDegrees=15, twistLimitDegrees=1}
 {bone1Name=dyn_cloth_r_05, bone2Name=dyn_cloth_04, swing1LimitDegrees=15, swing2LimitDegrees=15, twistLimitDegrees=1}
 {bone1Name=dyn_cloth_l_07, bone2Name=dyn_cloth_l_06, swing1LimitDegrees=20, swing2LimitDegrees=45, twistLimitDegrees=2}
]
