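# Animation Blueprint for the SwampFish backpack cosmetic, adding secondary motion to its loops, cloth and jar bones as the character moves.
# The input pose gets spine-driven root offsets, a world-space pelvis collider and Anim Dynamics on the loops, then is cached.
# bRigidBodyEnabled selects either a RigidBody simulation or a cheaper-looking Trail controller fallback from that cache.
# The event graph derives gravity and the movement, crouch and skydive alphas feeding those branches.
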AnimBlueprint /BRCosmetics/Accessories/FORT_Backpacks/Backpack_M_MED_SwampFish/Meshes/M_MED_SwampFish_Pack_AnimBP
target_skeleton=&/M_MED_SwampFish_Pack_Skeleton
graphs [
 AnimGraph {nodes [
   Copy Pose From Mesh() {bUseAttachedParent=true}
   n1 = Get BPTranslation() -> (BPTranslation:Vector)
   n2 = Get BPRotate() -> (BPRotate:Rotator)
   "Transform (Modify) Bone - Bone: dyn_cloth_03"(Alpha=1, Translation=@n1, Rotation=@n2, Scale="1,1,1") {
    rotationMode=BMM_Additive
    scaleMode=BMM_Additive
    scaleSpace=BCS_BoneSpace
    translationMode=BMM_Additive
    alphaBoolBlend {blendOption=Linear}
    boneToModify {boneName=dyn_cloth_03}
   }
   Output Pose(Result=Blend Poses by bool(BlendPose_0=Use cached pose 'RbanCache'(), BlendPose_1=Use cached pose 'trailCache'(), bActiveValue=Get bRigidBodyEnabled()) {blendType=Linear}) {layerGroup=None}
   n3 = "Transform (Modify) Bone - Bone: root"(ComponentPose=Local To Component(LocalPose=Input Pose()), Translation=@n1, Rotation=@n2) {
    rotationMode=BMM_Additive
    scaleMode=BMM_Additive
    scaleSpace=BCS_BoneSpace
    translationMode=BMM_Additive
    alphaBoolBlend {blendOption=Linear}
    boneToModify {boneName=root}
   }
   n4 = "Transform (Modify) Bone - Bone: pelvis_collider"(
    ComponentPose=@n3
    Alpha=1
    Translation=Get pelvis_tran() -> (pelvis_tran:Vector)
    Rotation=Get pelvis_rot() -> (pelvis_rot:Rotator)
    Scale="1,1,1"
   ) {
    rotationMode=BMM_Replace
    rotationSpace=BCS_WorldSpace
    translationMode=BMM_Replace
    translationSpace=BCS_WorldSpace
    alphaBoolBlend {blendOption=Linear}
    boneToModify {boneName=pelvis_collider}
   }
   n5 = "Transform (Modify) Bone - Bone: pack"(ComponentPose=@n4, Translation="0,1,0", Rotation="0,0,0") {
    rotationMode=BMM_Additive
    translationMode=BMM_Additive
    alphaBoolBlend {blendOption=Linear}
    boneToModify {boneName=pack}
    translation {y=1}
   }
   n6 = "Transform (Modify) Bone - Bone: root"(ComponentPose=@n5, Translation="0,-1,0", Rotation="0,0,0") {translationMode=BMM_Additive, alphaBoolBlend {blendOption=Linear}, boneToModify {boneName=root}, translation {y=-1}}
   n7 = Anim Dynamics - Bone: dyn_loop_01(ComponentPose=@n6, Alpha=0.5, ExternalForce="0,0,0") {
    angularDampingOverride=1
    angularSpringConstant=1500
    bAngularSpring=true
    bOverrideAngularDamping=true
    bUseGravityOverride=true
    simulationSpace=RootRelative
    alphaBoolBlend {blendOption=Linear}
    boundBone {boneName=dyn_loop_01}
    componentAppliedLinearAccClamp=<3000 3000 3000>
    componentLinearAccScale=<0.4 0.4 0.4>
    componentLinearVelScale=<0.4 0.4 0.4>
    gravityOverride {z=-980}
    physicsBodyDefinitions [{
      boundBone {boneName=dyn_loop_01}
      boxExtents=<10 10 10>
      constraintSetup {angularLimitsMax=<30 30 30>, angularLimitsMin=<-30 -30 -30>, angularTarget=<0 0 0>}
      localJointOffset=<2 0 0>
     }]
    relativeSpaceBone {boneName=pack}
   }
   n8 = Anim Dynamics - Bone: dyn_loop_02(ComponentPose=@n7, Alpha=0.6, ExternalForce="0,0,0") {
    angularDampingOverride=1
    angularSpringConstant=1000
    bAngularSpring=true
    bOverrideAngularDamping=true
    bUseGravityOverride=true
    simulationSpace=RootRelative
    alphaBoolBlend {blendOption=Linear}
    boundBone {boneName=dyn_loop_02}
    componentAppliedLinearAccClamp=<3000 3000 3000>
    componentLinearAccScale=<0.7 0.7 0.7>
    componentLinearVelScale=<0.7 0.7 0.7>
    gravityOverride {z=-980}
    physicsBodyDefinitions [{
      boundBone {boneName=dyn_loop_02}
      boxExtents=<10 10 10>
      constraintSetup {angularLimitsMax=<30 30 30>, angularLimitsMin=<-30 -30 -30>, angularTarget=<0 0 0>}
      localJointOffset=<2 0 0>
     }]
    relativeSpaceBone {boneName=pack}
   }
   n9 = Anim Dynamics - Bone: dyn_loop_03(ComponentPose=@n8, Alpha=0.8, ExternalForce="0,0,0") {
    angularDampingOverride=1
    angularSpringConstant=1000
    bAngularSpring=true
    bOverrideAngularDamping=true
    bUseGravityOverride=true
    simulationSpace=RootRelative
    alphaBoolBlend {blendOption=Linear}
    boundBone {boneName=dyn_loop_03}
    componentAppliedLinearAccClamp=<3000 3000 3000>
    componentLinearAccScale=<0.7 0.7 0.7>
    componentLinearVelScale=<0.7 0.7 0.7>
    gravityOverride {z=-980}
    physicsBodyDefinitions [{
      boundBone {boneName=dyn_loop_03}
      boxExtents=<10 10 10>
      constraintSetup {angularLimitsMax=<30 30 30>, angularLimitsMin=<-30 -30 -30>, angularTarget=<0 0 0>}
      localJointOffset=<2 0 0>
     }]
    relativeSpaceBone {boneName=pack}
   }
   n10 = Anim Dynamics - Bone: dyn_loop_04(ComponentPose=@n9, Alpha=0.4, ExternalForce="0,0,0") {
    angularDampingOverride=1
    angularSpringConstant=1000
    bAngularSpring=true
    bOverrideAngularDamping=true
    bUseGravityOverride=true
    simulationSpace=RootRelative
    alphaBoolBlend {blendOption=Linear}
    boundBone {boneName=dyn_loop_04}
    componentAppliedLinearAccClamp=<3000 3000 3000>
    componentLinearAccScale=<0.4 0.4 0.4>
    componentLinearVelScale=<0.4 0.4 0.4>
    gravityOverride {z=-980}
    physicsBodyDefinitions [{
      boundBone {boneName=dyn_loop_04}
      boxExtents=<10 10 10>
      constraintSetup {angularLimitsMax=<30 30 30>, angularLimitsMin=<-30 -30 -30>, angularTarget=<0 0 0>}
      localJointOffset=<3 0 0>
     }]
    relativeSpaceBone {boneName=pack}
   }
   Save cached pose 'Start_Cache'(Pose=Component To Local(ComponentPose=@n10)) {cacheName=Start_Cache}
   n11 = "Transform (Modify) Bone - Bone: root"!(
    ComponentPose=Local To Component(LocalPose=Use cached pose 'Start_Cache'())
    bAlphaBoolEnabled:bool=NOT Boolean(A=Get bRigidBodyEnabled())
    Translation="0,0,0"
    Rotation="0,0,-15"
    Scale
   ) -> (Pose) {
    alphaInputType=Bool
    rotationMode=BMM_Additive
    translationMode=BMM_Additive
    alphaBoolBlend {blendOption=Linear}
    alphaScaleBiasClamp {bInterpResult=true, interpSpeedDecreasing=3, interpSpeedIncreasing=2}
    boneToModify {boneName=root}
    rotation {roll=-15}
   }
   n12 = Get BwdMovement() -> (BwdMovement:double)
   n13 = Trail controller - Bone: dyn_cloth_04(ComponentPose=@n11, Alpha=@n12) {
    bUsePlanarLimit=true
    chainLength=3
    debugLifeTime=0.1
    maxDeltaTime=0.333
    alphaScaleBiasClamp {bInterpResult=true, scale=0.5}
    baseJoint {boneName=root}
    planarLimits [{drivingBone {boneName=root}, planeTransform {location=<0 -3 0>, rotation=<rot 0 0 -90>, scale=<1 1 1>}}]
    rotationLimits [{limitMax=<180 180 180>}, {limitMax=<180 180 180>}, {limitMax=<180 180 180>}]
    rotationOffsets [<0 0 0>, <0 0 0>, <0 0 0>]
    trailBone {boneName=dyn_cloth_04}
    trailRelaxationSpeed {editorCurveData {keys [<at -0.0134 19.3841 RCIM_Linear>,<at 0.5079 12.8986 RCIM_Linear>,<at 0.9955 9.6044 RCIM_Linear>]}}
   }
   n14 = Trail controller - Bone: dyn_jar_bottom_c_03(ComponentPose=@n13, Alpha=@n12) {
    bUsePlanarLimit=true
    chainLength=3
    debugLifeTime=0.1
    maxDeltaTime=0.333
    alphaScaleBiasClamp {bInterpResult=true, scale=0.5}
    baseJoint {boneName=root}
    planarLimits [{drivingBone {boneName=root}, planeTransform {location=<0 -3 0>, rotation=<rot 0 0 -90>, scale=<1 1 1>}}]
    rotationLimits [{limitMax=<180 180 180>}, {limitMax=<180 180 180>}, {limitMax=<180 180 180>}]
    rotationOffsets [<0 0 0>, <0 0 0>, <0 0 0>]
    trailBone {boneName=dyn_jar_bottom_c_03}
    trailRelaxationSpeed {editorCurveData {keys [<at -0.0134 19.3841 RCIM_Linear>,<at 0.5079 12.8986 RCIM_Linear>,<at 0.9955 9.6044 RCIM_Linear>]}}
   }
   Save cached pose 'trailCache'(Pose=Component To Local(ComponentPose=@n14)) {cacheName=trailCache}
   n15 = Get FlyPose_alpha() -> (FlyPose_alpha:double)
   n16 = "Transform (Modify) Bone - Bone: dyn_cloth_l_05"(
    ComponentPose=Local To Component(LocalPose=Use cached pose 'Start_Cache'())
    Alpha=@n15
    Translation="0,-2,0"
    Rotation="0,0,0"
    Scale="1,1,1"
   ) {
    translationMode=BMM_Additive
    alphaBoolBlend {blendOption=Linear}
    alphaScaleBiasClamp {bInterpResult=true, interpSpeedDecreasing=3, interpSpeedIncreasing=2}
    boneToModify {boneName=dyn_cloth_l_05}
    translation {y=-2}
   }
   n17 = "Transform (Modify) Bone - Bone: dyn_cloth_r_05"(ComponentPose=@n16, Alpha=@n15, Translation="0,-2,0", Rotation="0,0,0", Scale="1,1,1") {
    translationMode=BMM_Additive
    alphaBoolBlend {blendOption=Linear}
    alphaScaleBiasClamp {bInterpResult=true, interpSpeedDecreasing=3, interpSpeedIncreasing=2}
    boneToModify {boneName=dyn_cloth_r_05}
    translation {y=-2}
   }
   n18 = "Transform (Modify) Bone - Bone: root"!(ComponentPose=@n17, Alpha:float=Get Moving2D() -> (Moving2D:double), Translation="0,0,0", Rotation="0,0,10", Scale) -> (Pose) {
    rotationMode=BMM_Additive
    translationMode=BMM_Additive
    alphaBoolBlend {blendOption=Linear}
    alphaScaleBiasClamp {bInterpResult=true, interpSpeedDecreasing=7, interpSpeedIncreasing=5}
    boneToModify {boneName=root}
    rotation {roll=10}
   }
   n19 = RigidBody!(ComponentPose=@n18, Alpha, OverrideWorldGravity:Vector=Get GravityOverride(), ExternalForce) -> (Pose) {
    bOverrideWorldGravity=true
    overridePhysicsAsset=&/M_MED_SwampFish_Pack_Physics
    simulationSpace=BaseBoneSpace
    alphaBoolBlend {blendOption=Linear}
    baseBoneRef {boneName=root}
    componentAppliedLinearAccClamp=<5000 5000 5000>
    componentLinearAccScale=<0.4 0.4 0.4>
    componentLinearVelScale=<0.4 0.4 0.4>
    overrideWorldGravity {z=-980}
   }
   n20 = "Transform (Modify) Bone - Bone: root"(ComponentPose=@n19, Alpha=Get CrouchMove() -> (CrouchMove:double), Translation="0,-1,0", Rotation="0,0,0") {
    translationMode=BMM_Additive
    alphaBoolBlend {blendOption=Linear}
    alphaScaleBiasClamp {bInterpResult=true, interpSpeedDecreasing=3, interpSpeedIncreasing=2}
    boneToModify {boneName=root}
    translation {y=-1}
   }
   Save cached pose 'RbanCache'(Pose=Component To Local(ComponentPose=@n20)) {cacheName=RbanCache}
  ]}
 EventGraph {nodes [
   Event BlueprintUpdateAnimation()
   > n1 = Sequence!(execute) -> (then_0, then_1, then_2:Exec, then_3:Exec, then_4:Exec, then_5:Exec, then_6:Exec, then_7:Exec)
   n2 = GetOwnerSkeletalMeshComponent()
   n3 = BreakRotator(InRot=BreakTransform(InTransform=GetDeltaTransformFromRefPose(self=@n2, BoneName=spine_05, BaseName=spine_01)).Rotation)
   Set BPRotate(execute=@n1.then_1, BPRotate:Rotator=MakeRotator(Roll=MapRangeClamped(Value=@n3.Yaw, InRangeA=60, InRangeB=-30, OutRangeA=10, OutRangeB=25))) -> (Output_Get:Rotator)
   > Set BPTranslation(BPTranslation:Vector=MakeVector(Y=MapRangeClamped(Value=@n3.Yaw, InRangeA=-2, InRangeB=-200, OutRangeB=25))) -> (Output_Get:Vector)
   n4 = GetPartSkeletalMeshComponent()
   > Is Valid(InputObject=@n4)
   n5 = GetAttachParent!(self:Scene Component Object Reference=@n4) -> (ReturnValue:Scene Component Object Reference)
   > Is Valid(InputObject=@n5)
   n6 = BreakTransform(InTransform=GetSocketTransform(self=@n5, InSocketName=pelvis, TransformSpace=RTS_World))
   > Set pelvis_tran(pelvis_tran:Vector=@n6) -> (Output_Get:Vector)
   > Set pelvis_rot(pelvis_rot:Rotator=@n6.Rotation) -> (Output_Get:Rotator)
   n7 = PlayerMovement(execute=@n1.then_6, AnimInstance=Self-Reference(), XDotProductCoefficient=1, YDotProductCoefficient=1, ZDotProductCoefficient=1)
   > n8 = Branch(Condition=AND Boolean(A=Get bIsTargeting(), B=Get bIsCrouching()))
   n9 = MapRangeClamped(
    Value=float + float(A="Absolute (Float)"(A=@n7), B=MapRangeClamped(Value=@n7.YMovement, InRangeA=-1, InRangeB=1, OutRangeA=0.3)) {is=title K2Node_CommutativeAssociativeBinaryOperator}
    InRangeA=0
    InRangeB=1.65
   )
   > Set BwdMovement(BwdMovement:double=float * float(A=@n9, B=0.25) {is=title K2Node_PromotableOperator}) -> (Output_Get:double)
   Set BwdMovement(execute=@n8.else, BwdMovement=@n9)
   n10 = Branch(execute=@n1.then_7, Condition=Get bIsSkyDiving())
   > n11 = Branch(Condition="float > float"(A=Get PawnVelocityZ(), B=50) {is=title K2Node_CallFunction})
   > Set BwdMovement(BwdMovement=float * float(A=Get BwdMovement(), B=0) {is=title K2Node_CommutativeAssociativeBinaryOperator})
   Set BwdMovement(execute=[@n10.else, @n11.else], BwdMovement=float * float(A=Get BwdMovement(), B=1))
   n12 = Branch(execute=@n1.then_2, Condition=Get bIsSkydiveFloating())
   > Set FlyPose_alpha(FlyPose_alpha:double=1) -> (Output_Get:double)
   Set FlyPose_alpha(execute=@n12.else, FlyPose_alpha=0)
   n13 = Get bIsCrouching()
   n14 = Branch(execute=@n1.then_4, Condition=AND Boolean(A=Get bIsMoving2D(), B=NOT Boolean(A=@n13)))
   > Set Moving2D(Moving2D:double=1) -> (Output_Get:double)
   Set Moving2D(execute=@n14.else, Moving2D=0)
   n15 = Gravity Override State Function DEPRECATED(
    execute=@n1.then_0
    CharacterPartAnimInstance=Self-Reference()
    GameTimeInSeconds=GetGameTimeInSeconds()
    GravityOverrideParams=Get Gravity Override Params()
   )
   > Set GravityOverride(GravityOverride=@n15)
   n16 = Branch(execute=@n1.then_3, Condition=Get bIsSprinting())
   > Set Sprinting_alpha(Sprinting_alpha:double=1) -> (Output_Get:double)
   Set Sprinting_alpha(execute=@n16.else, Sprinting_alpha=0)
   n17 = Branch(execute=@n1.then_5, Condition=AND Boolean(A=Get bIsMoving2D(), B=Get bIsCrouching()))
   > Set CrouchMove(CrouchMove:double=1) -> (Output_Get:double)
   Set CrouchMove(execute=@n17.else, CrouchMove=0)
   Event BlueprintInitializeAnimation()
   > Set SkelMesh(SkelMesh=GetOwnerSkeletalMeshComponent())
   n18 = TryGetPawnOwner()
   > Is Valid(InputObject=@n18)
   > n19 = Cast To FortPlayerPawnAthena(Object=@n18)
   > Set PlayerPawn(PlayerPawn:Fort Player Pawn Athena Object Reference=@n19) -> (Output_Get:Fort Player Pawn Athena Object Reference)
   Event OnBeginSkydiving()
   > Set Skydiving(Skydiving:double=1) -> (Output_Get:double)
   > Set SkydivingAlpha(SkydivingAlpha:Rotator="0,-65,0") -> (Output_Get:Rotator)
   Event OnEndSkydiving()
   > Set Skydiving(Skydiving=0)
   > Set SkydivingAlpha(SkydivingAlpha="0,-10,0")
  ]}
]
variables [
 backpack_rotate {type=Rotator}
 backpack_rotate_cloth {type=Rotator}
 backpack_translation {type=Vector}
 BPRotate {type=Rotator}
 BPTranslation {type=Vector}
 BwdMovement {type=number}
 CrouchMove {type=number}
 FlyPose_alpha {type=number}
 Gravity Override Params {type=GravityOverrideParamsStruct}
 GravityOverride {type=Vector}
 Moving2D {type=number}
 pelvis_rot {type=Rotator}
 pelvis_tran {type=Vector}
 PlayerPawn {type=FortPlayerPawnAthena}
 SkelMesh {type=SkeletalMeshComponent}
 Skydiving {type=number}
 SkydivingAlpha {type=Rotator}
 Sprinting_alpha {type=number}
]
